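# Post-process material that renders the view down a sniper scope, with lens distortion, reticle and glass reflection.
# The AttachScope function supplies UVs for a circular lens mask, inside which scene color is refracted by a sphere normal map, darkened by lens AO and tinted.
# A reticle texture and cubemap reflection, tinted by ambient cubemap color, are layered on with a soft vignette.
# BlendAmount scales both the refraction and this final layer.
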Material /Game/Weapons/FORT_Rifles/SniperScope/DownScopePostProcess
graph {
 properties {blendMode=BLEND_Translucent, materialDomain=MD_PostProcess, pixelDepthOffsetMode=PDOM_Legacy, usedShadingModels=Default Lit}
 nodes [
  n1 = ComponentMask(None=MaterialFunctionCall!(RotateDialMask, Normal WS) {materialFunction=/Game/Packages/Fortress_SharedMaterials/Base_Material_Functions/AttachScope}.Post process uvs 0-1 uvs) {g=true, r=true}
  n2 = Constant() {r=0.5}
  n3 = Distance(A=@n1, B=@n2)
  n4 = TextureSample(UVs=MaterialFunctionCall(UVs=@n1, Texture Scale=Constant() {r=1.15}) {materialFunction=&ScaleUVsByCenter}.UVs) {samplerType=SAMPLERTYPE_LinearColor, texture=/Game/Packages/Fortress_SharedMaterials/Textures/SphereNormalsUncomp}
  n5 = ConstantBiasScale(None=@n4.RGB) {bias=-0.5, scale=2}
  n6 = Multiply(
   A=Multiply(A=ScalarParameter() {defaultValue=1, parameterName=BlendAmount}, B=ScalarParameter() {defaultValue=-0.1, parameterName=RefractionAmount})
   B=ComponentMask(None=Multiply(A=@n3, B=@n5)) {g=true, r=true}
  )
  n7 = TextureCoordinate()
  n8 = ComponentMask(None=SceneTexture(UVs=Add(A=@n6, B=@n7)) {sceneTextureId=PPI_PostProcessInput0}.Color) {b=true, g=true, r=true}
  n9 = ScalarParameter() {defaultValue=0.035, parameterName=Lens AO Radius}
  n10 = LinearInterpolate(
   B=ScalarParameter() {defaultValue=0.4431, parameterName=Lens AO Value, sliderMax=1}
   Alpha=Power(Base=SphereMask(A=@n3, B=@n2, Radius=@n9) {hardnessPercent=0})
  ) {constA=1, constB=0}
  n11 = Constant() {r=0.5}
  n12 = SphereMask(A=@n1, B=@n11, Radius=@n11) {hardnessPercent=99}
  n13 = LinearInterpolate(
   A=ComponentMask(None=SceneTexture() {sceneTextureId=PPI_PostProcessInput0}.Color) {b=true, g=true, r=true}
   B=Multiply(A=Multiply(A=@n8, B=@n10), B=VectorParameter() {parameterName=Lens Coloration, defaultValue=<rgba 0.9075 1.0679 1.1 0.25>}.RGB)
   Alpha=@n12
  )
  n14 = LinearInterpolate(
   A=@n13
   B=VectorParameter() {parameterName=Reticle Color, defaultValue=<rgba 0 0 0 1>}.RGB
   Alpha=Multiply(
    A=TextureSampleParameter2D(UVs=@n1) {parameterName=Reticle, samplerType=SAMPLERTYPE_LinearColor, texture=/Game/Athena/HUD/reticles/HUD_SniperCrossHairs_v1_Athena}.R
    B=@n12
   )
  ) {constB=0}
  n15 = MaterialFunctionCall() {materialFunction=/Engine/Functions/GetPostProcessFunctions/GetAmbientCubemapTint}
  n16 = Subtract(A=@n4.RGB) {constB=0.5}
  n17 = Constant3Vector() {constant=<rgba 2 -2 2 0>}
  MaterialExpressionTextureSampleParameterCube_0 = TextureSampleParameterCube!(UVs=Transform(None=Multiply(A=@n16, B=@n17)) {transformSourceType=TRANSFORMSOURCE_View}, Apply View MipBias) {parameterName=MaterialExpressionTextureSampleParameterCube_0, texture=/Game/Packages/Fortress_Sky/TexturesHDR/T_hdrsky_test03_low}
  n18 = Multiply(
   A=@n12
   B=Multiply(A=ComponentMask(None=@n15.AmbientCubemapTint) {b=true, g=true, r=true}, B=@MaterialExpressionTextureSampleParameterCube_0.RGB)
  )
  n19 = Power(Base=OneMinus(None=@n4.B)) {constExponent=1.5}
  n20 = Multiply(A=@n10, B=Multiply(A=Multiply(A=@n18, B=@n19), B=ScalarParameter() {defaultValue=4.0551, parameterName=Scope Reflection Intensity}))
  n21 = Constant() {r=0.5}
  LinearInterpolate(
   A=@n13
   B=Multiply(A=Add(A=@n14, B=@n20), B=SphereMask(A=@n21, B=@n1) {attenuationRadius=2, hardnessPercent=0})
   Alpha=ScalarParameter() {defaultValue=1, parameterName=BlendAmount}
  )
  MaterialFunctionCall() {materialFunction=/Engine/Functions/GetPostProcessFunctions/GetAmbientCubemapIntensity}
  TextureSample() {texture=/Game/Packages/Fortress_SharedMaterials/Textures/T_Trench_Ref01}
  TextureSample() {texture=/Game/Environments/Props/Rocks/Textures/T_Props_CrystalEnv_Black}
  TextureSample() {texture=/Game/Environments/Props/Rocks/Textures/T_Props_CrystalEnv_Green}
  TextureSample() {texture=/Game/Environments/Props/Rocks/Textures/T_Props_CrystalEnv_Purple}
  TextureSample() {texture=/Game/Environments/Props/Rocks/Textures/T_Props_CrystalEnv_White}
  TextureSample() {texture=/Game/Packages/Fortress_SharedMaterials/Textures/T_Trench_Ref02}
  TextureSample() {texture=/Game/Environments/Props/Rocks/Textures/T_Props_CrystalEnv_Red}
  TextureSample() {texture=/Game/Environments/Props/Rocks/Textures/T_Props_CrystalEnv_Yellow}
  TextureSample() {texture=/Game/Packages/Fortress_SharedMaterials/Textures/T_Trench_Ref03}
  TextureSample() {texture=/Game/Environments/Props/Rocks/Textures/T_Props_CrystalEnv_Rainbow}
  TextureSample() {texture=/Game/Packages/Fortress_SharedMaterials/Textures/T_Trench_Ref05}
 ]
}
