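# Master material for rifle sniper scopes, Default Lit with world-space normals and flagged for Nanite use.
# A UV box mask marks the rotating dial region and feeds the shared AttachScope function, along with the Normal map transformed out of tangent space.
# BaseColor and SpecularMasks texture parameters, an inverted dial mask and a stray constant are present, with no connections shown from them.
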Material /Game/Weapons/FORT_Rifles/SniperScope/M_GlobalScope_Master
graph {
 properties {bTangentSpaceNormal=false, bUsedWithNanite=true, pixelDepthOffsetMode=PDOM_Legacy, usedShadingModels=Default Lit}
 nodes [
  Constant() {r=2}
  n1 = MaterialFunctionCall!(A=TextureCoordinate(), B=Constant2Vector() {g=0.15, r=0.25}, Bounds=Constant2Vector() {g=0.2, r=0.5}, Edge Falloff=Constant()) {materialFunction=/Engine/Functions/Engine_MaterialFunctions01/Gradient/BoxMask-2D}
  MaterialFunctionCall!(
   RotateDialMask=@n1.Result
   Normal WS=Transform(None=TextureSampleParameter2D() {parameterName=Normal, samplerType=SAMPLERTYPE_Normal, texture=&FlatNormalBlue}.RGB)
  ) {materialFunction=/Game/Packages/Fortress_SharedMaterials/Base_Material_Functions/AttachScope}
  OneMinus(None=@n1.Result)
  TextureSampleParameter2D() {parameterName=BaseColor, texture=&T_BlacksRGB}
  TextureSampleParameter2D() {parameterName=SpecularMasks, samplerType=SAMPLERTYPE_Masks, texture=&T_BlackMask}
 ]
}
