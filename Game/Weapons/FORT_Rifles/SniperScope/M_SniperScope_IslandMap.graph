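# Post-process material for a sniper scope that overlays an island minimap with storm circles on the scene.
# Minimap UVs are aspect-corrected, zoomed by current-circle size and have a stepped sine-pulse glitch variant.
# MPC_Minimap drives the current and next circles, pulsing hot/cold next-next circle, player dot and rotated indicator arrow.
# Desaturation, tint, gridlines and scanlines style the map, and a sphere mask helps blend it over scene color.
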Material /Game/Weapons/FORT_Rifles/SniperScope/M_SniperScope_IslandMap
graph {
 properties {materialDomain=MD_PostProcess, pixelDepthOffsetMode=PDOM_Legacy, usedShadingModels=Default Lit}
 nodes [
  n1 = Constant3Vector() {desc=Arrow Color, constant=<rgba 0.8681 0.3021 2 0>}
  n2 = NamedRerouteUsage()
  n3 = SphereMask(A=NamedRerouteUsage(), B=Constant() {r=0.5}, Radius=Constant() {r=0.535}, Hardness=Constant() {r=0.95})
  LinearInterpolate(
   A=ComponentMask(None=SceneTexture() {sceneTextureId=PPI_PostProcessInput0}.Color) {b=true, g=true, r=true}
   B=Multiply(A=LinearInterpolate(A=@n1, B=@n2, Alpha=@n3), B=NamedRerouteUsage())
   Alpha=Add(A=@n3, B=NamedRerouteUsage())
  )
  n4 = MaterialFunctionCall(Float2=ViewProperty() {property=MEVP_ViewSize}.Property) {materialFunction=&BreakOutFloat2Components}
  n5 = Divide(A=@n4.R, B=@n4.R)
  n6 = Constant() {r=1}
  n7 = Multiply(
   A=NamedRerouteDeclaration(None=AppendVector(A=@n5, B=@n6)) {name=ScreenRatios, nodeColor=<rgba 1 0 0.7969 1>}
   B=Add(A=ScreenPosition().ViewportUV, B=Constant2Vector() {g=-0.225, r=0.38})
  )
  n8 = Frac(None=@n5)
  n9 = NamedRerouteDeclaration(None=AppendVector(A=Multiply(A=@n8) {constB=-0.5}, B=Constant())) {name=ScreenOffset, nodeColor=<rgba 0 1 0.625 1>}
  NamedRerouteDeclaration(None=MaterialFunctionCall(UVs=Add(A=@n7, B=@n9), Texture Scale=Constant() {r=0.22}) {materialFunction=&ScaleUVsByCenter}.UVs) {name=UVs, nodeColor=<rgba 0.6172 1 0 1>}
  n10 = ScreenPosition()
  n11 = Time()
  n12 = Constant() {r=16}
  n13 = Divide(A=Round(None=Multiply(A=@n11, B=@n12)), B=@n12)
  n14 = Panner(Coordinate=@n10.ViewportUV, Time=@n13) {bFractionalPart=true, speedY=0.176}
  n15 = Panner(Coordinate=@n10.ViewportUV, Time=@n13) {bFractionalPart=true, speedY=-0.12}
  n16 = Add(A=Sine(None=ComponentMask(None=@n14) {g=true}) {period=0.1536}, B=Sine(None=ComponentMask(None=@n15) {g=true}) {period=0.0538})
  n17 = Constant() {r=0.5}
  n18 = Divide(A=Round(None=Multiply(A=@n16, B=@n17)), B=@n17)
  n19 = Constant()
  n20 = Add(A=Sine(None=@n13), B=Sine(None=@n13) {period=1.5})
  n21 = Power(Base=Divide(A=Abs(None=@n20))) {constExponent=15}
  n22 = Multiply(
   A=Multiply(A=AppendVector(A=@n18, B=@n19), B=NamedRerouteDeclaration(None=@n21) {name=SinePulse, nodeColor=<rgba 0 1 0.2266 1>})
   B=Constant() {r=0.03}
  )
  NamedRerouteDeclaration(None=Add(A=NamedRerouteUsage(), B=@n22)) {name=UVsDistorted, nodeColor=<rgba 1 0 0.1641 1>}
  n23 = CollectionParameter() {collection=/Game/Athena/HUD/MiniMap/MPC_Minimap, parameterName=CurrentCircle}
  n24 = ComponentMask(None=@n23) {b=true}
  n25 = Saturate(None=Multiply(A=@n24) {constB=3.33})
  NamedRerouteDeclaration(None=@n25) {name=CurrentCircle.B.Scaled, nodeColor=<rgba 0 1 0.4141 1>}
  n26 = ScreenPosition()
  n27 = NamedRerouteUsage()
  n28 = Add(A=Multiply(A=NamedRerouteUsage(), B=Add(A=@n26.ViewportUV, B=@n27)), B=NamedRerouteUsage())
  n29 = Constant2Vector() {g=0.5, r=0.5}
  n30 = Multiply(A=CollectionParameter() {collection=&MPC_Minimap, parameterName=NextNextCircleAngleToPlayer}) {constB=-1}
  n31 = Constant() {r=0.1}
  n32 = MaterialFunctionCall(Float2=MaterialFunctionCall(
    UVs=MaterialFunctionCall(UVs=@n28, Rotation Center=@n29, "Rotation Angle (0-1)"=@n30) {materialFunction=&CustomRotator}.Rotated Values
    Texture Scale=AppendVector(A=@n31, B=@n31)
   ) {materialFunction=&ScaleUVsByCenter}.UVs) {materialFunction=&BreakOutFloat2Components}
  n33 = ScalarParameter() {defaultValue=4.2, parameterName=ShiftYTargetIndicator}
  n34 = TextureSample(UVs=AppendVector(A=@n32.R, B=Add(A=@n32.G, B=@n33))) {
   samplerSource=SSM_Clamp_WorldGroupSettings
   samplerType=SAMPLERTYPE_LinearColor
   texture=&/T_UI_WeatherSniper_ScopeIndicator_Optimized
  }
  NamedRerouteDeclaration(None=Saturate(None=Add(A=@n34.R, B=@n34.R))) {name=Arrow, nodeColor=<rgba 0 1 0.2734 1>}
  NamedRerouteDeclaration(None=@n22) {name=UVDistortion_Isolated, nodeColor=<rgba 1 0 0.9844 1>}
  n35 = Multiply(A=Rotator(Coordinate=ScreenPosition().ViewportUV, Time=Constant() {r=0.7853}) {speed=1}) {constB=220}
  n36 = MaterialFunctionCall(Float2=Frac(None=Panner(Coordinate=@n35) {bFractionalPart=true, speedY=1.88})) {materialFunction=&BreakOutFloat2Components}
  n37 = Constant() {r=0.5}
  n38 = Constant() {r=0.25}
  NamedRerouteDeclaration(None=ConstantBiasScale(None=SphereMask(A=@n36.G, B=@n37, Radius=@n38)) {bias=6, scale=0.2}) {name=Scanlines, nodeColor=<rgba 0 1 0.5078 1>}
  n39 = Multiply(
   A=NamedRerouteUsage()
   B=VectorParameter() {group=EffectParams, parameterName=TintOutsideSafeZone, defaultValue=<rgba 1.4911 0.5104 2 0.1>}.RGB
  )
  n40 = CollectionParameter() {collection=&MPC_Minimap, parameterName=NextCircle}
  n41 = Distance(A=NamedRerouteUsage(), B=ComponentMask(None=@n40) {g=true, r=true})
  n42 = LinearInterpolate(A=Constant() {r=0.5}, B=ScalarParameter() {defaultValue=20, parameterName=Next-ScaleFactor}, Alpha=NamedRerouteUsage())
  n43 = Multiply(A=Constant() {r=0.0005}, B=@n42)
  n44 = ComponentMask(None=@n40) {b=true}
  n45 = MaterialFunctionCall(Alpha=@n41, Min=Subtract(A=@n44, B=@n43), Max=@n44) {materialFunction=&SmoothStep}
  n46 = CollectionParameter() {collection=&MPC_Minimap, parameterName=CurrentCircle}
  n47 = Distance(A=NamedRerouteUsage(), B=ComponentMask(None=@n46) {g=true, r=true})
  n48 = Constant() {r=5}
  n49 = ScalarParameter() {defaultValue=40, parameterName=Current-ScaleFactor}
  n50 = NamedRerouteUsage()
  n51 = ComponentMask(None=@n46) {b=true}
  n52 = Subtract(A=@n51, B=Multiply(A=Constant() {r=0.0005}, B=LinearInterpolate(A=@n48, B=@n49, Alpha=@n50)))
  n53 = Add(A=@n45.Result)
  n54 = Add(A=MaterialFunctionCall(Alpha=@n41, Min=@n44, Max=Add(A=@n44, B=@n43)) {materialFunction=&SmoothStep}.Result)
  n55 = Add(
   A=LinearInterpolate(
    A=NamedRerouteUsage()
    B=Multiply(A=@n39, B=@n45.Result)
    Alpha=MaterialFunctionCall(Alpha=@n47, Min=@n52, Max=@n51) {materialFunction=&SmoothStep}.Result
   )
   B=Multiply(A=Constant3Vector() {constant=<rgba 1 0.9594 0 0>}, B=Subtract(A=@n53, B=@n54))
  )
  n56 = LinearInterpolate(
   A=Constant3Vector() {constant=<rgba 0.4566 0.4566 0.4566 0>}
   B=Multiply(A=Constant3Vector() {constant=<rgba 0.9843 0 1 0>}) {constB=4}
   Alpha=CollectionParameter() {collection=&MPC_Minimap, parameterName=NextNextCircleHotOrCold}
  ) {constA=0.5, constB=1.3}
  n57 = CollectionParameter() {collection=&MPC_Minimap, parameterName=NextNextCircle}
  n58 = Distance(A=NamedRerouteUsage(), B=ComponentMask(None=@n57) {g=true, r=true})
  n59 = Multiply(A=Constant() {r=0.0005}, B=@n42)
  n60 = ComponentMask(None=@n57) {b=true}
  n61 = Subtract(A=@n60, B=@n59)
  n62 = Add(A=@n60, B=@n59)
  n63 = Subtract(
   A=Add(A=MaterialFunctionCall(Alpha=@n58, Min=@n61, Max=@n60) {materialFunction=&SmoothStep}.Result)
   B=Add(A=MaterialFunctionCall(Alpha=@n58, Min=@n60, Max=@n62) {materialFunction=&SmoothStep}.Result)
  )
  n64 = LinearInterpolate(
   A=Multiply(A=Multiply(A=@n56, B=@n63) {constB=0.2}) {constB=0.2}
   B=Multiply(A=@n56, B=@n63)
   Alpha=MaterialFunctionCall(Speed=Constant() {r=0.7}) {materialFunction=&DebugTimeSine}.Result
  ) {constA=0.2}
  NamedRerouteDeclaration(None=LinearInterpolate(A=Add(A=@n55, B=@n64), B=Constant3Vector() {constant=<rgba 1 0.0913 0.0087 0>}, Alpha=NamedRerouteUsage())) {name=UpdatedWeatherScope, nodeColor=<rgba 0.2656 1 0 1>}
  n65 = NamedRerouteUsage()
  NamedRerouteDeclaration(None=MaterialFunctionCall!(UVsOverride=Saturate(None=@n65), TextureOverride) {materialFunction=/Game/Athena/Apollo/Maps/UI/MF_Minimap}.RGB) {name=MiniMapTexture, nodeColor=<rgba 1 0 0.8203 1>}
  n66 = Multiply(A=NamedRerouteUsage()) {constB=0.2}
  n67 = Add(B=NamedRerouteUsage()) {constA=0.125}
  n68 = Add(A=Desaturation(None=Add(A=@n66, B=@n67), Fraction=Constant() {r=0.8}), B=NamedRerouteUsage())
  NamedRerouteDeclaration(None=Multiply(A=Constant3Vector() {constant=<rgba 0.5949 1.1573 1.8 0>}, B=@n68)) {name=MiniMapColorAdjusted, nodeColor=<rgba 1 0.1406 0 1>}
  NamedRerouteDeclaration(None=@n68) {name=MiniMap, nodeColor=<rgba 0.8984 0 1 1>}
  n69 = If(A=@n24, B=Constant() {r=0.08}, "A > B"=Constant2Vector() {g=2.3, r=0.73}, "A < B"=Constant2Vector() {g=5, r=2.6})
  n70 = LinearInterpolate(A=ComponentMask(None=@n69) {g=true}, B=ComponentMask(None=@n69) {r=true}, Alpha=@n25)
  n71 = NamedRerouteUsage()
  n72 = ComponentMask(None=@n23) {g=true, r=true}
  n73 = MaterialFunctionCall(Float2=Subtract(A=Constant2Vector() {g=0.5, r=0.5}, B=@n72)) {materialFunction=&BreakOutFloat2Components}
  n74 = AppendVector(A=Multiply(A=@n73.R) {constB=-1}, B=Multiply(A=@n73.G) {constB=-1})
  n75 = Constant() {r=512}
  n76 = Multiply(A=@n72, B=@n75) {constB=512}
  NamedRerouteDeclaration(None=MaterialFunctionCall!(Texture Scale=AppendVector(A=@n70, B=@n70), UVs=Add(A=@n71, B=@n74), Offset XY Texturesize Z=AppendVector(A=@n76, B=@n75)) {materialFunction=/Engine/Functions/Engine_MaterialFunctions01/Texturing/ScaleUVsAroundPoint}.UVs) {name=ScaledUVs, nodeColor=<rgba 1 0 0.3984 1>}
  n77 = Frac(None=Multiply(A=NamedRerouteUsage()) {constB=5})
  n78 = MaterialFunctionCall(Float2=Abs(None=ConstantBiasScale(None=@n77) {bias=-0.5, scale=2})) {materialFunction=&BreakOutFloat2Components}
  n79 = SphereMask(A=Max(A=@n78.R, B=@n78.R), B=Constant() {r=1}, Radius=Constant() {r=0.12}, Hardness=Constant() {r=0.05})
  NamedRerouteDeclaration(None=Multiply(A=Power(Base=@n79) {constExponent=3}, B=Constant() {r=0.07})) {name=Gridlines, nodeColor=<rgba 1 0 0.8437 1>}
  n80 = CollectionParameter() {collection=&MPC_Minimap, parameterName=PlayerLocation}
  n81 = Constant() {r=0.002}
  n82 = Constant() {r=0.02}
  n83 = NamedRerouteUsage()
  NamedRerouteDeclaration(None=SphereMask(A=NamedRerouteUsage(), B=ComponentMask(None=@n80) {g=true, r=true}, Radius=LinearInterpolate(A=@n81, B=@n82, Alpha=@n83))) {name=PlayerDot, nodeColor=<rgba 1 0 0 1>}
 ]
}
