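# Prototype masked, Default Lit, Nanite-enabled material for the sniper scope mesh, with world-space normals.
# The shared AttachScope function supplies an opacity that is raised to a power, boosted, clamped and run through DitherTemporalAA for a dithered cutout.
# Roughness blends between min and max parameters using a channel of UV-scaled tiling noise, while a ShadingPathSwitch substitutes a single scalar on mobile.
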Material /Game/Weapons/FORT_Rifles/SniperScope/M_SniperScope_MeshMaterial_proto
graph {
 properties {bTangentSpaceNormal=false, bUsedWithNanite=true, blendMode=BLEND_Masked, pixelDepthOffsetMode=PDOM_Legacy, usedShadingModels=Default Lit}
 nodes [
  Constant() {r=1}
  Constant() {r=0.05}
  n1 = Power(Base=MaterialFunctionCall!(RotateDialMask, Normal WS) {materialFunction=/Game/Packages/Fortress_SharedMaterials/Base_Material_Functions/AttachScope}.Opacity) {constExponent=3}
  MaterialFunctionCall(Alpha Threshold=Clamp(None=Multiply(A=@n1) {constB=20})) {materialFunction=&DitherTemporalAA}
  n2 = ScalarParameter() {defaultValue=3, parameterName=Roughness_U_Scale}
  n3 = ScalarParameter() {defaultValue=3, parameterName=Roughness_V_Scale, sortPriority=1}
  n4 = TextureSampleParameter2D(UVs=Multiply(A=TextureCoordinate(), B=AppendVector(A=@n2, B=@n3))) {parameterName=Roughness Texture, texture=/Engine/MapTemplates/TilingNoise05}
  n5 = AppendVector(A=@n4.RGB, B=@n4.RGB)
  ShadingPathSwitch(
   Default=LinearInterpolate(
    A=ScalarParameter() {defaultValue=0.4, parameterName=Roughness_Min}
    B=ScalarParameter() {defaultValue=0.8, parameterName=Roughness_Max}
    Alpha=StaticComponentMaskParameter(None=@n5) {defaultR=true, parameterName=Roughness Mask}
   ) {constA=0.4, constB=0.8}
   Mobile=ScalarParameter() {defaultValue=0.5048, parameterName=Roughness_Mobile}
  ) {desc=Migrated from FeatureLevelSwitch}
 ]
}
