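# MF_Fisheye_Distortion is a material function in the sniper scope folder that computes a radial UV offset for fisheye distortion.
# It raises the distance from Coords to Center to an Exponent, multiplies that by the normalized direction away from Center, and scales it by Amount.
# Coords defaults to viewport UV and Center to mid-screen.
# The output is only the offset, not the distorted coordinates.
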MaterialFunction /Game/Weapons/FORT_Rifles/SniperScope/MF_Fisheye_Distortion
graph {nodes [
  n1 = FunctionInput() {bUsePreviewValueAsDefault=true, inputName=Center, inputType=FunctionInput_Vector2, sortPriority=200, previewValue=<0.5 0.5 0 1>}
  n2 = FunctionInput(Preview=ScreenPosition().ViewportUV) {bUsePreviewValueAsDefault=true, inputName=Coords, inputType=FunctionInput_Vector2, sortPriority=50}
  n3 = FunctionInput() {bUsePreviewValueAsDefault=true, inputName=Exponent, inputType=FunctionInput_Scalar, sortPriority=1000, previewValue=<3 0 0 1>}
  n4 = Power(Base=Distance(A=@n2, B=@n1), Exp=@n3) {constExponent=3}
  n5 = Normalize(VectorInput=Subtract(A=@n2, B=@n1))
  n6 = FunctionInput() {bUsePreviewValueAsDefault=true, inputName=Amout, inputType=FunctionInput_Scalar, sortPriority=600, previewValue=<0.2 0 0 1>}
  FunctionOutput(None=Multiply(A=Multiply(A=@n4, B=@n5), B=@n6) {constB=0.2})
 ]}
