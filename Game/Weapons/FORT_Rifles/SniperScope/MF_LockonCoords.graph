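# Material function for the sniper scope that outputs a 2D offset for a lock-on target's screen position.
# It reads Target Screen X and Y from the MPC_Lockon parameter collection and subtracts them from 0.5, giving the offset from target to screen center, assuming normalized coordinates.
# The Is Targeting parameter blends the result from zero to that offset.
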MaterialFunction /Game/Weapons/FORT_Rifles/SniperScope/MF_LockonCoords
graph {nodes [
  n1 = AppendVector(
   A=CollectionParameter() {collection=&/MPC_Lockon, parameterName=Target Screen X}
   B=CollectionParameter() {collection=&MPC_Lockon, parameterName=Target Screen Y}
  )
  FunctionOutput(None=LinearInterpolate(
    B=Subtract(B=@n1) {constA=0.5, constB=0.5}
    Alpha=CollectionParameter() {collection=&MPC_Lockon, parameterName=Is Targeting}
   ))
 ]}
