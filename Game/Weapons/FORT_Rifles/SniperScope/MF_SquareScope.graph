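# Material function for a sniper scope that outputs remapped screen UVs, fitting a square layout that a 0-1 ratio input widens toward full screen.
# Centered viewport UVs are aspect-corrected using buffer size, and a horizontal three-zone mask is built from them.
# The mask blends left-shifted, unshifted and right-shifted UVs, clamping the sides so the middle stretches.
# The hue-shift and derivative nodes do not feed the output.
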MaterialFunction /Game/Weapons/FORT_Rifles/SniperScope/MF_SquareScope
graph {nodes [
  n1 = FunctionInput() {
   bUsePreviewValueAsDefault=true
   desc="0 gives square, 1 covers screen"
   inputName="Rectangle Ratio (0-1)"
   inputType=FunctionInput_Scalar
   previewValue=<0.5 0 0 1>
  }
  n2 = AppendVector(A=LinearInterpolate(B=NamedRerouteUsage(), Alpha=@n1), B=Constant())
  n3 = ViewProperty() {property=MEVP_BufferSize}
  n4 = Divide(A=ComponentMask(None=@n3.Property) {r=true}, B=ComponentMask(None=@n3.Property) {g=true})
  n5 = Multiply(A=Subtract(A=ScreenPosition().ViewportUV) {constB=0.5}, B=AppendVector(A=@n4, B=Constant() {r=1}))
  n6 = Add(A=@n5) {constB=0.5}
  n7 = MaterialFunctionCall!(
   1=Clamp(None=Subtract(A=@n6, B=@n2), Max=Constant2Vector() {g=1, r=0.3}) {clampMode=CMODE_ClampMax, maxDefault=0.3}
   2=@n6
   3=Clamp(None=Add(A=@n6, B=@n2), Min=Constant2Vector() {r=0.7}) {clampMode=CMODE_ClampMin, maxDefault=0.3}
   4
   A=NamedRerouteUsage()
  ) {materialFunction=/Engine/Functions/Engine_MaterialFunctions02/Lerp_Multiple_Float2}
  n8 = ComponentMask(None=@n7.Lerp 3 Inputs) {r=true}
  DDX(Value=@n8)
  DDY(Value=@n8)
  FunctionOutput(None=@n7.Lerp 3 Inputs)
  MaterialFunctionCall(Hue Shift Percentage=Multiply(A=@n8) {constB=2.3}, Texture=Constant3Vector() {constant=<rgba 1 0 0 0>}) {materialFunction=&HueShift}
  n9 = Multiply(A=FunctionInput() {bUsePreviewValueAsDefault=true, inputType=FunctionInput_Scalar, previewValue=<0.6 0 0 1>})
  n10 = Multiply(A=@n9) {constB=-1}
  n11 = ComponentMask(None=@n5) {r=true}
  n12 = Floor(None=Multiply(A=InverseLinearInterpolate(A=@n10, B=@n9, Value=@n11) {constA=-1.5}) {constB=3})
  NamedRerouteDeclaration(None=Saturate(None=Divide(A=@n12))) {name=3 Zones Mask, nodeColor=<rgba 0 0.0859 1 1>}
  NamedRerouteDeclaration(None=Add(A=Multiply(B=@n4) {constA=-0.5}) {constB=0.5}) {name=UV Corner Transform, nodeColor=<rgba 1 0.3984 0 1>}
 ]}
