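# Post-process material for a sniper scope view, layering a tintable reticle and glowing crosshair over the scene with vignette shading.
# Aspect-corrected screen UVs sample the channel-packed reticle, with switchable fisheye, square-scope stretch and MPC_WeaponScope camera sway.
# The crosshair optionally follows lock-on coordinates, using separate core and adjustable blur masks with additive or lerped color.
# The vignette switches between radial fade and rounded-rectangle distance field, optionally shifted by camera motion.
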Material /Game/Weapons/FORT_Rifles/SniperScope/M_SniperScope_Post_s19
graph {
 properties {materialDomain=MD_PostProcess, pixelDepthOffsetMode=PDOM_Legacy, usedShadingModels=Default Lit}
 nodes [
  Multiply(A=NamedRerouteUsage(), B=NamedRerouteUsage())
  n1 = CollectionParameter() {collection=&/MPC_WeaponScope, parameterName=Camera Movement}
  n2 = ScalarParameter() {defaultValue=200, group=Camera Motion, parameterName=Camera Motion Dampening Vignette}
  n3 = Multiply(A=@n2) {constB=-1}
  n4 = StaticSwitchParameter(True=Divide(A=ComponentMask(None=@n1) {g=true, r=true}, B=AppendVector(A=@n2, B=@n3)) {constB=100}, False=Constant2Vector()) {group=Camera Motion, parameterName=UseCameraMotionVignette}
  n5 = Add(A=@n4, B=NamedRerouteUsage())
  n6 = ScalarParameter() {defaultValue=0.5, parameterName=Rec Vignette X}
  n7 = ScalarParameter() {defaultValue=0.25, parameterName=Rec Vignette Y}
  n8 = ScalarParameter() {defaultValue=0.2, desc=r, parameterName=Rectangular Vignette Radius}
  n9 = Add(A=Subtract(A=Abs(None=@n5), B=AppendVector(A=@n6, B=@n7)), B=@n8)
  n10 = Min(A=Max(A=ComponentMask(None=@n9) {desc=q.x, r=true}, B=ComponentMask(None=@n9) {desc=q.y, g=true})) {constB=0}
  n11 = MaterialFunctionCall(Vector 2=Max(A=@n9) {constB=0, desc="max(q,0)"}) {materialFunction=/Engine/Functions/Engine_MaterialFunctions02/Utility/VectorLength}
  n12 = Saturate(None=Subtract(A=Add(A=@n10, B=@n11.V2 Length), B=@n8))
  n13 = Constant()
  n14 = ScalarParameter() {defaultValue=0.5, parameterName=Rec Vignette Radius}
  n15 = ScalarParameter() {parameterName=Rec Vignette Hardness}
  n16 = ScalarParameter() {group=Side Fade, parameterName=Edge Darkness}
  n17 = Add(A=NamedRerouteUsage(), B=@n4)
  n18 = Constant()
  n19 = ScalarParameter() {defaultValue=0.66, group=Side Fade, parameterName=Fade Center Scale}
  n20 = ScalarParameter() {group=Side Fade, parameterName=Fade Center Hardness}
  n21 = Saturate(None=Power(
    Base=SphereMask(A=@n17, B=@n18, Radius=@n19, Hardness=@n20)
    Exp=ScalarParameter() {defaultValue=1, group=Side Fade, parameterName=Fade Center Power}
   ))
  NamedRerouteDeclaration(None=StaticSwitchParameter(True=SphereMask(A=@n12, B=@n13, Radius=@n14, Hardness=@n15), False=LinearInterpolate(A=@n16, Alpha=@n21) {constA=25}) {parameterName=Use Rectangular Vignette}) {name=Vignette, nodeColor=<rgba 0.2 0.2 0.2 1>}
  n22 = NamedRerouteUsage()
  n23 = StaticSwitchParameter(True=Add(A=NamedRerouteUsage(), B=@n22), False=@n22) {group=Reticle Fisheye, parameterName=UseCrosshairFisheyeDistortion, sortPriority=34}
  n24 = Add(
   A=Multiply(A=MaterialFunctionCall() {materialFunction=&/MF_LockonCoords}.Result, B=NamedRerouteUsage())
   B=@n23
  )
  n25 = TextureSampleParameter2D(UVs=MaterialFunctionCall(
    UVs=StaticSwitchParameter(True=@n24, False=@n23) {parameterName=Use Lockon Coords}
    Texture Scale=ScalarParameter() {defaultValue=0.25, group=Crosshair, parameterName=Crosshair_Scale}
   ) {materialFunction=&ScaleUVsByCenter}.UVs) {
   group=Crosshair
   parameterName=T_Crosshair
   samplerSource=SSM_Clamp_WorldGroupSettings
   samplerType=SAMPLERTYPE_LinearColor
   texture=/Game/Weapons/FORT_Rifles/Textures/Crosshair_Dot_M
  }
  n26 = Add(
   A=Subtract(A=@n25.G)
   B=Clamp(None=ScalarParameter() {defaultValue=0.5, group=Crosshair, parameterName=CrosshairBlurAmount}) {maxDefault=0.98}
  )
  n27 = Multiply(A=NamedRerouteUsage(), B=@n26)
  NamedRerouteDeclaration(None=Saturate(None=StaticSwitchParameter(True=@n27, False=@n26) {parameterName=Use Mask Crosshair})) {name=Crosshair Mask Blur, nodeColor=<rgba 0.2656 0 1 1>}
  n28 = NamedRerouteUsage()
  n29 = NamedRerouteUsage()
  n30 = LinearInterpolate(
   A=VectorParameter() {group=Crosshair, parameterName=Crosshair_BlurColor2, defaultValue=<rgba 0.4115 0 0 1>}.RGB
   B=VectorParameter() {group=Crosshair, parameterName=Crosshair_BlurColor, defaultValue=<rgba 1 0 0 1>}.RGB
   Alpha=@n29
  )
  n31 = ScalarParameter() {defaultValue=1, group=Crosshair, parameterName=CrosshairBlurGlowIntensity}
  n32 = Multiply(
   A=@n29
   B=LinearInterpolate(
    A=Multiply(A=@n30, B=@n31)
    B=VectorParameter() {group=Crosshair, parameterName=Crosshair_CoreColor, defaultValue=<rgba 0.1005 0 2 1>}.RGB
    Alpha=NamedRerouteUsage()
   )
  )
  n33 = LinearInterpolate(
   A=NamedRerouteUsage()
   B=VectorParameter() {group=Crosshair, parameterName=Crosshair_BlurColor, defaultValue=<rgba 1 0 0 1>}.RGB
   Alpha=Saturate(None=NamedRerouteUsage())
  )
  n34 = VectorParameter() {group=Crosshair, parameterName=Crosshair_CoreColor, defaultValue=<rgba 0.1005 0 2 1>}
  n35 = NamedRerouteUsage()
  NamedRerouteDeclaration(None=StaticSwitchParameter(True=Add(A=@n28, B=@n32), False=LinearInterpolate(A=@n33, B=@n34.RGB, Alpha=@n35)) {group=Crosshair, parameterName=Additive Crosshair}) {name=Crosshair, nodeColor=<rgba 0.6641 1 0 1>}
  n36 = SceneTexture() {sceneTextureId=PPI_PostProcessInput0}
  n37 = NamedRerouteUsage()
  n38 = ScalarParameter() {defaultValue=0.5, group=Reticle, parameterName=ReticleBlurAmount}
  n39 = Max(A=NamedRerouteUsage(), B=Add(A=Subtract(A=@n37), B=Clamp(None=@n38) {maxDefault=0.98}))
  n40 = SphereMask(A=NamedRerouteUsage(), B=Constant(), Radius=ScalarParameter() {defaultValue=0.5, parameterName=Fade Center MNask}, Hardness=Constant() {r=1})
  NamedRerouteDeclaration(None=LinearInterpolate(
    A=ComponentMask(None=@n36.Color) {b=true, g=true, r=true}
    B=VectorParameter() {group=Reticle, parameterName=Reticle Color, defaultValue=<rgba 0 0 0 1>}.RGB
    Alpha=Multiply(A=@n39, B=@n40)
   )) {name=Scene Reticle, nodeColor=<rgba 1 0.5 0.5 1>}
  n41 = MaterialFunctionCall(Float2=MaterialFunctionCall() {materialFunction=&ScreenResolution}.Visible Resolution) {materialFunction=&BreakOutFloat2Components}
  n42 = Divide(A=@n41.R, B=@n41.R)
  NamedRerouteDeclaration(None=AppendVector(A=@n42, B=Constant() {r=1})) {name=Screen Ratio, nodeColor=<rgba 1 0.1 0.1 1>}
  n43 = VectorParameter() {parameterName=CrosshairOffset, defaultValue=<rgba 0 0 0 1>}
  n44 = ScreenPosition()
  n45 = MaterialFunctionCall(Float2=Add(
    A=MaterialFunctionCall(X=@n43.R, Y=@n43.R) {materialFunction=&MakeFloat2}.Result
    B=Add(A=@n44.ViewportUV) {constB=-0.5}
   )) {materialFunction=&BreakOutFloat2Components}
  n46 = MaterialFunctionCall(X=Multiply(A=@n45.R, B=@n42), Y=@n45.G) {materialFunction=&MakeFloat2}
  NamedRerouteDeclaration(None=Add(A=NamedRerouteDeclaration(None=@n46.Result) {name=Screen Position Centered, nodeColor=<rgba 1 0.1 0.1 1>}) {constB=0.5}) {name=Screen Position Ratioed, nodeColor=<rgba 1 0.1 0.1 1>}
  n47 = MaterialFunctionCall!(
   Coords
   Center=AppendVector(
    A=ScalarParameter() {defaultValue=0.5, group=Reticle Fisheye, parameterName=Reticle Fisheye Center X, sortPriority=50}
    B=ScalarParameter() {defaultValue=0.5, group=Reticle Fisheye, parameterName=Reticle Fisheye Center Y, sortPriority=51}
   )
   Amout=ScalarParameter() {defaultValue=0.2, group=Reticle Fisheye, parameterName=Reticle Fisheye Amount, sortPriority=100}
   Exponent=ScalarParameter() {defaultValue=3, group=Reticle Fisheye, parameterName=Reticle Fisheye Exponent, sortPriority=200}
  ) {materialFunction=&/MF_Fisheye_Distortion}
  n48 = StaticSwitchParameter(
   True=MaterialFunctionCall!(In, "Rectangle Ratio (0-1)"=ScalarParameter() {group=Reticle, parameterName=Square Scope Stretch, sliderMax=1}) {materialFunction=&/MF_SquareScope}.Result
   False=NamedRerouteUsage()
  ) {parameterName=UseSquareScope}
  n49 = StaticSwitchParameter(True=Add(A=NamedRerouteDeclaration(None=@n47.Result) {name=Fisheye Distortion, nodeColor=<rgba 0 0.577 1 1>}, B=@n48), False=@n48) {group=Reticle Fisheye, parameterName=UseReticleFisheyeDistortion, sortPriority=33}
  n50 = ScalarParameter() {defaultValue=200, group=Camera Motion, parameterName=Camera Motion Dampening}
  n51 = Divide(
   A=ComponentMask(None=CollectionParameter() {collection=&MPC_WeaponScope, parameterName=Camera Movement}) {g=true, r=true}
   B=AppendVector(A=@n50, B=Multiply(A=@n50) {constB=-1})
  ) {constB=100}
  n52 = TextureSampleParameter2D(UVs=StaticSwitchParameter(True=Add(A=@n49, B=@n51), False=@n49) {group=Camera Motion, parameterName=UseCameraMotion}) {group=Reticle, parameterName=T_Reticle, samplerType=SAMPLERTYPE_LinearColor, texture=&/Reticle_M}
  NamedRerouteDeclaration(None=@n52.R) {name=Reticle R, nodeColor=<rgba 1 0 0.2 1>}
  NamedRerouteDeclaration(None=@n52.G) {name=Reticle G, nodeColor=<rgba 0 1 0.2 1>}
  NamedRerouteDeclaration(None=@n52.B) {name=Reticle B, nodeColor=<rgba 0 0.2 1 1>}
  n53 = NamedRerouteUsage()
  NamedRerouteDeclaration(None=StaticSwitchParameter(True=Multiply(A=@n25.R, B=@n53), False=@n25.R) {parameterName=Use Mask Crosshair}) {name=Crosshair Mask Core, nodeColor=<rgba 0 1 0.4141 1>}
 ]
}
