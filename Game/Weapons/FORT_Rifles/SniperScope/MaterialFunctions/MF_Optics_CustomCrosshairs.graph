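# Material function that serves as the custom crosshair stage for sniper scope optics, currently wrapping a single crosshair effect, MF_MarkExecuteSniper.
# Scene texture, optics mask, crosshair UV and scope UV inputs each feed a color-coded named reroute.
# The nested function takes scene texture, mask and crosshair UV through reroutes, and its result is the sole output.
# Scope UVs are declared but have no matching pin on that call.
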MaterialFunction /Game/Weapons/FORT_Rifles/SniperScope/MaterialFunctions/MF_Optics_CustomCrosshairs
graph {nodes [
  FunctionOutput(None=MaterialFunctionCall!(Scene Texture=NamedRerouteUsage(), Mask=NamedRerouteUsage(), Crosshair UV=NamedRerouteUsage()) {materialFunction=&/MF_MarkExecuteSniper}.Result)
  NamedRerouteDeclaration(None=FunctionInput() {bUsePreviewValueAsDefault=true, inputName=Scene Texture, sortPriority=10}) {name=Input Scene Texture, nodeColor=<rgba 0.8281 1 0 1>}
  NamedRerouteDeclaration(None=FunctionInput() {bUsePreviewValueAsDefault=true, inputName=Optics Mask, inputType=FunctionInput_Scalar, sortPriority=1000}) {name=Optics Mask, nodeColor=<rgba 0.8516 0 1 1>}
  NamedRerouteDeclaration(None=FunctionInput() {bUsePreviewValueAsDefault=true, inputName=Crosshair UVs, inputType=FunctionInput_Vector2, sortPriority=2000}) {name=Crosshair UVs, nodeColor=<rgba 1 0.6094 0 1>}
  NamedRerouteDeclaration(None=FunctionInput() {bUsePreviewValueAsDefault=true, inputName=Scope UVs, inputType=FunctionInput_Vector2, sortPriority=3000}) {name=Scope UVs, nodeColor=<rgba 0.2891 0 1 1>}
 ]}
