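# Parent post-process material for rifle optics, drawing scope zoom, vignette and reticle during aim-down-sights.
# MF_FirstPersonScope supplies eye UVs, masks and camera distance, which drive scene-UV magnification and reticle placement.
# The reticle supports SDF edges, fisheye, reload follow, thermal-toggled colors, custom crosshairs and spread-driven five-tap mip-biased blur.
# An optional SM6 custom HLSL radial blur adds dithered depth of field, switchable for performance, alongside debug views.
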Material /Game/Weapons/FORT_Rifles/SniperScope/PP_GlobalOpticMods_Parent
graph {
 properties {materialDomain=MD_PostProcess, pixelDepthOffsetMode=PDOM_Legacy, usedShadingModels=Default Lit}
 nodes [
  ComponentMask() {g=true, r=true}
  n1 = Multiply(A=Time(), B=ScalarParameter() {defaultValue=0.25, parameterName=DebugSpeed})
  Divide(A=Add(A=Sine(None=@n1)))
  n2 = ComponentMask(None=VectorParameter() {parameterName=PlayerVelocity, defaultValue=<rgba 0 0 0 1>}.RGB) {b=true, g=true, r=true}
  Divide(
   A=Subtract(A=Length(None=@n2) {is=class MaterialExpressionLength}, B=ScalarParameter() {defaultValue=1, parameterName=MinVelocityCrossHairBlur})
   B=ScalarParameter() {defaultValue=100, parameterName=MaxVelocityCrossHairBlur}
  )
  n3 = MaterialFunctionCall!(
   Plane Depth Offset=ScalarParameter() {group=Optics Masks, parameterName=Offset}
   Scope Eye Mask Hardness=Constant()
   Scale=ScalarParameter() {defaultValue=2.2, group=Optics Masks, parameterName=Scope Scale}
   Scope Target Point WS
   Scope Eye Point WS
   Up Vector
   In
  ) {materialFunction=/Game/Packages/Fortress_SharedMaterials/Base_Material_Functions/MF_FirstPersonScope}
  InverseLinearInterpolate(
   A=ScalarParameter() {defaultValue=225, group=ADS Transition, parameterName=BlendDistanceStart}
   B=ScalarParameter() {defaultValue=50, group=ADS Transition, parameterName=BlendDistanceEnd}
   Value=@n3.Camera - Eye Point Distance
  )
  n4 = Power(Base=SmoothStep(
    Min=ScalarParameter() {group=Zoom, parameterName=Refraction Remap Min, sliderMax=1}
    Max=ScalarParameter() {defaultValue=0.2, group=Zoom, parameterName=Refraction Remap Max, sliderMax=1}
    Value=NamedRerouteUsage()
   )) {constExponent=0.5}
  n5 = NamedRerouteUsage()
  n6 = LinearInterpolate(
   A=ScalarParameter() {defaultValue=1, group=Zoom, parameterName=Optical Zoom Min, sliderMax=2, sliderMin=1}
   B=ScalarParameter() {defaultValue=1, group=Zoom, parameterName=Optical Zoom Max, sliderMax=2, sliderMin=1}
   Alpha=Multiply(A=NamedRerouteDeclaration(None=@n4) {name=Refraction Debug, nodeColor=<rgba 0.176 1 0 1>}, B=Power(Base=@n5) {constExponent=4})
  ) {constA=1}
  n7 = NamedRerouteUsage()
  n8 = Constant() {r=1}
  LinearInterpolate(B=@n6, Alpha=Multiply(A=SphereMask(A=@n7, B=@n8) {attenuationRadius=0.2, hardnessPercent=0}, B=NamedRerouteUsage())) {constA=1}
  LinearInterpolate(Alpha=Custom!(None) {code=View.GPUSceneViewId, showCode=true})
  n9 = VectorParameter() {parameterName=CrosshairOffset, defaultValue=<rgba 0 0 0 1>}
  MaterialFunctionCall(X=@n9.R, Y=@n9.R) {materialFunction=&MakeFloat2}
  n10 = NamedRerouteUsage()
  Max(A=@n10, B=NamedRerouteUsage())
  n11 = NamedRerouteUsage()
  n12 = Subtract(A=Abs(None=Subtract(A=@n11) {constB=0.5}), B=ScalarParameter() {defaultValue=0.51, parameterName=UVEdgeMask_size})
  n13 = ScalarParameter() {defaultValue=0.2, parameterName=UVEdgeMask_Blend}
  n14 = MaterialFunctionCall(Float2=Saturate(None=Divide(A=@n12, B=@n13))) {materialFunction=&BreakOutFloat2Components}
  n15 = OneMinus(None=Max(A=@n14.R, B=@n14.R))
  n16 = TextureSampleParameter2D!(UVs=NamedRerouteUsage(), Bias=NamedRerouteUsage(), Apply View MipBias) {
   group=Crosshair
   mipValueMode=TMVM_MipBias
   parameterName=T_Crosshair
   samplerSource=SSM_Clamp_WorldGroupSettings
   samplerType=SAMPLERTYPE_LinearColor
   texture=/Game/Weapons/FORT_Rifles/Textures/Crosshair_Dot_M
  }
  n17 = ScalarParameter() {desc="Set by Code when Player Moves and Weapon becomes \r\nless accurate 0-1 ", parameterName=SpreadRatio}
  n18 = NamedRerouteUsage()
  n19 = LinearInterpolate(
   A=Saturate(None=Divide(A=@n17) {constB=1.5})
   Alpha=Multiply(A=OneMinus(None=@n18), B=ScalarParameter() {defaultValue=1, parameterName=CrossHairBlur_FadeBlurMask})
  ) {constB=0}
  n20 = MaterialFunctionCall!(OffsetAmount=LinearInterpolate(B=ScalarParameter() {defaultValue=0.001, group=Blur based on Accuracy, parameterName=MaxBlurSampleOffset}, Alpha=@n19)) {materialFunction=/Engine/Functions/Engine_MaterialFunctions01/ImageAdjustment/BlurSampleOffsets}
  n21 = TextureSampleParameter2D(UVs=Add(A=@n11, B=@n20."-.5,.5"), Bias=NamedRerouteUsage()) {
   group=Crosshair
   mipValueMode=TMVM_MipBias
   parameterName=T_Crosshair
   samplerSource=SSM_Clamp_WorldGroupSettings
   samplerType=SAMPLERTYPE_LinearColor
   texture=&Crosshair_Dot_M
  }
  n22 = Add(A=@n16.RGBA, B=@n21.RGBA)
  n23 = TextureSampleParameter2D(UVs=Add(A=@n11, B=@n20.".5,.5"), Bias=NamedRerouteUsage()) {
   group=Crosshair
   mipValueMode=TMVM_MipBias
   parameterName=T_Crosshair
   samplerSource=SSM_Clamp_WorldGroupSettings
   samplerType=SAMPLERTYPE_LinearColor
   texture=&Crosshair_Dot_M
  }
  n24 = TextureSampleParameter2D(UVs=Add(A=@n20.".5,-.5", B=@n11), Bias=NamedRerouteUsage()) {
   group=Crosshair
   mipValueMode=TMVM_MipBias
   parameterName=T_Crosshair
   samplerSource=SSM_Clamp_WorldGroupSettings
   samplerType=SAMPLERTYPE_LinearColor
   texture=&Crosshair_Dot_M
  }
  n25 = TextureSampleParameter2D(UVs=Add(A=@n20."-.5,-.5", B=@n11), Bias=NamedRerouteUsage()) {
   group=Crosshair
   mipValueMode=TMVM_MipBias
   parameterName=T_Crosshair
   samplerSource=SSM_Clamp_WorldGroupSettings
   samplerType=SAMPLERTYPE_LinearColor
   texture=&Crosshair_Dot_M
  }
  n26 = Divide(A=Add(A=Add(A=@n22, B=@n23.RGBA), B=Add(A=@n24.RGBA, B=@n25.RGBA))) {constB=5}
  n27 = ScalarParameter() {defaultValue=1, group=Blur based on Accuracy, parameterName=CrossHairBlur_FadeOpacity}
  n28 = LinearInterpolate(A=@n26, B=LinearInterpolate(A=@n26, B=Multiply(A=@n26, B=@n27), Alpha=NamedRerouteUsage()), Alpha=@n19)
  Multiply(A=@n15, B=@n28)
  n29 = ScalarParameter() {defaultValue=0.5, group=Reticle Fisheye, parameterName=Reticle Fisheye Center X, sortPriority=50}
  n30 = ScalarParameter() {defaultValue=0.5, group=Reticle Fisheye, parameterName=Reticle Fisheye Center Y, sortPriority=51}
  NamedRerouteDeclaration(None=MaterialFunctionCall!(
    Coords
    Center=AppendVector(A=@n29, B=@n30)
    Amout=ScalarParameter() {defaultValue=0.2, group=Reticle Fisheye, parameterName=Reticle Fisheye Amount, sortPriority=100}
    Exponent=ScalarParameter() {defaultValue=3, group=Reticle Fisheye, parameterName=Reticle Fisheye Exponent, sortPriority=200}
   ) {materialFunction=&/MF_Fisheye_Distortion}.Result) {name=Fisheye Distortion, nodeColor=<rgba 0 0.577 1 1>}
  n31 = Subtract(A=NamedRerouteUsage()) {constB=0.5}
  n32 = ScalarParameter() {defaultValue=3, group=Vignette, parameterName=Rec Vignette X}
  n33 = ScalarParameter() {defaultValue=3, group=Vignette, parameterName=Rec Vignette Y}
  n34 = ScalarParameter() {defaultValue=0.001, desc=r, group=Vignette, parameterName=Rectangular Vignette Radius}
  n35 = Add(A=Subtract(A=Abs(None=@n31), B=AppendVector(A=@n32, B=@n33)), B=@n34)
  n36 = ComponentMask(None=@n35) {desc=q.x, r=true}
  n37 = ComponentMask(None=@n35) {desc=q.y, g=true}
  n38 = Add(
   A=Min(A=Max(A=@n36, B=@n37)) {constB=0}
   B=MaterialFunctionCall(Vector 2=Max(A=@n35) {constB=0, desc="max(q,0)"}) {materialFunction=/Engine/Functions/Engine_MaterialFunctions02/Utility/VectorLength}.V2 Length
  )
  n39 = SphereMask(
   A=Saturate(None=Subtract(A=@n38, B=@n34))
   B=Constant()
   Radius=ScalarParameter() {defaultValue=0.5, group=Vignette, parameterName=Rec Vignette Radius}
   Hardness=ScalarParameter() {group=Vignette, parameterName=Rec Vignette Hardness}
  )
  n40 = SphereMask(
   A=Subtract(A=NamedRerouteUsage()) {constB=0.5}
   B=Constant()
   Radius=LinearInterpolate(
    A=ScalarParameter() {defaultValue=15, group=Vignette, parameterName=Fade Center Scale Start}
    B=ScalarParameter() {defaultValue=4, group=Vignette, parameterName=Fade Center Scale}
    Alpha=NamedRerouteUsage()
   )
   Hardness=ScalarParameter() {group=Vignette, parameterName=Fade Center Hardness}
  )
  n41 = ScalarParameter() {defaultValue=1, group=Vignette, parameterName=Fade Center Power}
  n42 = LinearInterpolate(A=ScalarParameter() {group=Vignette, parameterName=Edge Darkness}, Alpha=Saturate(None=Power(Base=@n40, Exp=@n41))) {constA=25}
  n43 = OneMinus(None=NamedRerouteUsage())
  n44 = NamedRerouteUsage()
  NamedRerouteDeclaration(None=LinearInterpolate(B=StaticSwitchParameter(True=@n39, False=@n42) {group=Vignette, parameterName=Use Rectangular Vignette}, Alpha=Multiply(A=@n43, B=@n44)) {constA=1}) {name=Vignette, nodeColor=<rgba 0.2 0.2 0.2 1>}
  n45 = NamedRerouteUsage()
  n46 = NamedRerouteUsage()
  n47 = NamedRerouteUsage()
  n48 = NamedRerouteUsage()
  n49 = NamedRerouteUsage()
  n50 = NamedRerouteUsage()
  n51 = NamedRerouteUsage()
  n52 = NamedRerouteUsage()
  n53 = ComponentMask(None=LinearInterpolate(A=LinearInterpolate(A=@n49, B=@n50, Alpha=@n51), B=LinearInterpolate(A=@n49, B=@n50, Alpha=@n52), Alpha=NamedRerouteUsage())) {b=true, g=true, r=true}
  n54 = Multiply(A=NamedRerouteUsage(), B=NamedRerouteUsage())
  n55 = LinearInterpolate(A=NamedRerouteUsage(), B=ComponentMask(None=NamedRerouteUsage()) {b=true, g=true, r=true}, Alpha=NamedRerouteUsage())
  n56 = Add(A=Multiply(A=@n53, B=@n54), B=@n55)
  n57 = NamedRerouteUsage()
  n58 = LinearInterpolate(A=@n55, B=@n53, Alpha=@n54) {constA=1}
  NamedRerouteDeclaration(None=StaticSwitchParameter(
    True=MaterialFunctionCall!(Scene Texture=@n45, Optics Mask=@n46, Crosshair UVs=@n47, Scope UVs=@n48) {materialFunction=/Game/Weapons/FORT_Rifles/SniperScope/MaterialFunctions/MF_Optics_CustomCrosshairs}.Result
    False=LinearInterpolate(A=@n56, B=@n58, Alpha=@n57)
   ) {group=Custom Crosshair, parameterName=Use Custom Crosshair}) {name=Crosshair, nodeColor=<rgba 0.6641 1 0 1>}
  n59 = TextureSampleParameter2D!(UVs=Saturate(None=NamedRerouteDeclaration(None=@n3.Scope Eye UV) {name=Scope Eye UV, nodeColor=<rgba 0.7813 0 1 1>}), Apply View MipBias) {
   group=Optics Masks
   parameterName=Mask Texture
   samplerType=SAMPLERTYPE_Masks
   texture=&/T_HolographicMask
  }
  n60 = NamedRerouteDeclaration(None=@n3.Scope Eye Mask) {name=Refraction Mask, nodeColor=<rgba 0 1 0.4609 1>}
  NamedRerouteDeclaration(None=StaticSwitchParameter(True=@n59.R, False=SmoothStep(Value=@n60) {constMax=0.01}) {group=Optics Masks, parameterName=UseMaskTexture, sortPriority=1}) {name=Optics Mask, nodeColor=<rgba 1 0.75 0 1>}
  n61 = NamedRerouteUsage()
  n62 = Multiply(A=ComponentMask(None=@n61) {r=true}, B=ComponentMask(None=NamedRerouteUsage()) {a=true})
  n63 = Multiply(A=ComponentMask(None=@n61) {g=true}, B=ComponentMask(None=NamedRerouteUsage()) {a=true})
  NamedRerouteDeclaration(None=Max(
    A=NamedRerouteDeclaration(None=@n62) {name=Reticle Mask Core, nodeColor=<rgba 0 1 0.4141 1>}
    B=NamedRerouteDeclaration(None=@n63) {name=Reticle Mask Glow, nodeColor=<rgba 0.1016 0 1 1>}
   )) {name=Reticle Mask, nodeColor=<rgba 0 0.5078 1 1>}
  NamedRerouteDeclaration(None=StaticSwitchParameter(True=@n59.G, False=@n60) {group=Optics Masks, parameterName=UseMaskTexture, sortPriority=1}) {name=Optics Mask Refraction, nodeColor=<rgba 0 1 0.3414 1>}
  NamedRerouteDeclaration(None=@n59.B) {name=Optics Mask Blur, nodeColor=<rgba 0.1016 0 1 1>}
  n64 = NamedRerouteUsage()
  n65 = SceneTexture() {sceneTextureId=PPI_PostProcessInput0}
  n66 = NamedRerouteUsage()
  n67 = LinearInterpolate(
   A=LinearInterpolate(A=ComponentMask(None=@n65.Color) {b=true, g=true, r=true}, B=Constant3Vector() {constant=<rgba 0 0 1 0>}, Alpha=OneMinus(None=@n66))
   B=Constant3Vector() {constant=<rgba 1 0 0 0>}
   Alpha=NamedRerouteUsage()
  )
  n68 = Constant3Vector()
  n69 = NamedRerouteUsage()
  n70 = Constant3Vector() {constant=<rgba 0 1 0 0>}
  n71 = StaticSwitchParameter(True=NamedRerouteUsage(), False=Add(A=LinearInterpolate(A=@n67, B=@n68, Alpha=@n10), B=Multiply(A=@n69, B=@n70))) {group=Debug, parameterName=Debug DOF}
  NamedRerouteDeclaration(None=StaticSwitchParameter(True=NamedRerouteUsage(), False=StaticSwitchParameter(True=@n64, False=@n71) {group=Debug, parameterName=Debug Reticle}) {group=Debug, parameterName=Debug Reload Reticle Size}) {name=DebugView, nodeColor=<rgba 0 1 0.3438 1>}
  n72 = ComponentMask(None=@n61) {b=true}
  n73 = NamedRerouteUsage()
  n74 = NamedRerouteUsage()
  NamedRerouteDeclaration(None=Multiply(A=Multiply(A=@n72, B=@n73), B=ComponentMask(None=@n74) {a=true})) {name=Reticle Secondary Mask, nodeColor=<rgba 0 1 0.5781 1>}
  n75 = VectorParameter() {group=Crosshair, parameterName=Crosshair_GlowColor, defaultValue=<rgba 0 0 0 1>}
  n76 = AppendVector(A=@n75.RGB, B=@n75.RGB)
  n77 = VectorParameter() {group=Toggleable Reticle, parameterName=Crosshair_GlowColorToggled, defaultValue=<rgba 0 0 0 1>}
  n78 = AppendVector(A=@n77.RGB, B=@n77.RGB)
  n79 = ScalarParameter() {parameterName=ToggleThermalMask}
  NamedRerouteDeclaration(None=StaticSwitchParameter(True=LinearInterpolate(A=@n76, B=@n78, Alpha=@n79), False=@n76) {group=Toggleable Reticle, parameterName=Enable Toggleable Reticle, sortPriority=5}) {name=Crosshair Glow Color, nodeColor=<rgba 1 0.4219 0 1>}
  n80 = VectorParameter() {group=Crosshair, parameterName=Crosshair_CoreColor, defaultValue=<rgba 0 0 0 1>}
  n81 = AppendVector(A=@n80.RGB, B=@n80.RGB)
  n82 = VectorParameter() {group=Toggleable Reticle, parameterName=Crosshair_CoreColorToggled, defaultValue=<rgba 0 0 0 1>}
  n83 = AppendVector(A=@n82.RGB, B=@n82.RGB)
  n84 = ScalarParameter() {parameterName=ToggleThermalMask}
  NamedRerouteDeclaration(None=StaticSwitchParameter(True=LinearInterpolate(A=@n81, B=@n83, Alpha=@n84), False=@n81) {group=Toggleable Reticle, parameterName=Enable Toggleable Reticle, sortPriority=5}) {name=Crosshair Core Color, nodeColor=<rgba 1 0.4219 0 1>}
  n85 = VectorParameter() {group=Crosshair, parameterName=Crosshair_SecondaryColor, defaultValue=<rgba 0 0 0 1>}
  n86 = AppendVector(A=@n85.RGB, B=@n85.RGB)
  n87 = VectorParameter() {group=Toggleable Reticle, parameterName=Crosshair_SecondaryColorToggled, defaultValue=<rgba 0 0 0 1>}
  n88 = AppendVector(A=@n87.RGB, B=@n87.RGB)
  n89 = ScalarParameter() {parameterName=ToggleThermalMask}
  NamedRerouteDeclaration(None=StaticSwitchParameter(True=LinearInterpolate(A=@n86, B=@n88, Alpha=@n89), False=@n86) {group=Toggleable Reticle, parameterName=Enable Toggleable Reticle, sortPriority=5}) {name=Crosshair Secondary Color, nodeColor=<rgba 1 0.4219 0 1>}
  n90 = MaterialFunctionCall(
   X=ScalarParameter() {defaultValue=1, group=Crosshair, parameterName=Reticle Opacity, sliderMax=1}
   Y=ScalarParameter() {group=Crosshair, parameterName=Core/Glow Hard Edge, sliderMax=1}
  ) {materialFunction=&MakeFloat4}
  n91 = ScalarParameter() {defaultValue=1, group=Toggleable Reticle, parameterName=Reticle Opacity Toggled, sliderMax=1}
  n92 = ScalarParameter() {group=Toggleable Reticle, parameterName=Core/Glow Hard Edge Toggled, sliderMax=1}
  n93 = StaticSwitchParameter(
   True=LinearInterpolate(
    A=@n90.Result
    B=MaterialFunctionCall(X=@n91, Y=@n92) {materialFunction=&MakeFloat4}.Result
    Alpha=ScalarParameter() {parameterName=ToggleThermalMask}
   )
   False=@n90.Result
  ) {group=Toggleable Reticle, parameterName=Enable Toggleable Reticle, sortPriority=5}
  NamedRerouteDeclaration(None=ComponentMask(None=@n93) {r=true}) {name=Reticle Opacity, nodeColor=<rgba 1 0.2344 0 1>}
  NamedRerouteDeclaration(None=ComponentMask(None=@n93) {g=true}) {name=Core/Glow Hard Edge, nodeColor=<rgba 1 0.2344 0 1>}
  n94 = NamedRerouteUsage()
  n95 = NamedRerouteUsage()
  n96 = ScalarParameter() {defaultValue=1, group=ADS Reload, parameterName=Reload Scale Multiplier}
  n97 = LinearInterpolate(
   A=NamedRerouteUsage()
   B=NamedRerouteDeclaration(None=MaterialFunctionCall(UVs=@n95, Texture Scale=@n96) {materialFunction=&ScaleUVsByCenter}.UVs) {name=Scope Eye UV Scaled, nodeColor=<rgba 1 0.2109 0 1>}
   Alpha=LinearInterpolate(
    B=ScalarParameter() {defaultValue=0.1, group=ADS Reload, parameterName=FollowScopeBias, sliderMax=1}
    Alpha=ScalarParameter() {group=C++ Parameters, parameterName=ScopedReloadWeight}
   ) {constAlpha=0.1}
  ) {constAlpha=0.1}
  n98 = Saturate(None=InverseLinearInterpolate(Value=NamedRerouteUsage()) {constA=100, constB=35})
  n99 = MaterialFunctionCall(
   UVs=StaticSwitchParameter(True=Add(A=@n94, B=@n97), False=@n97) {group=Reticle Fisheye, parameterName=UseCrosshairFisheyeDistortion, sortPriority=34}
   Texture Scale=Multiply(
    A=LinearInterpolate(Alpha=@n98) {constA=0.1}
    B=ScalarParameter() {defaultValue=0.25, group=Crosshair, parameterName=Crosshair_Scale, sliderMax=1}
   )
  ) {materialFunction=&ScaleUVsByCenter}
  n100 = TextureSampleParameter2D(UVs=NamedRerouteDeclaration(None=@n99.UVs) {name=CrossHairUVs, nodeColor=<rgba 0 1 0.9766 1>}) {
   group=Crosshair
   parameterName=T_Crosshair
   samplerSource=SSM_Clamp_WorldGroupSettings
   samplerType=SAMPLERTYPE_LinearColor
   texture=&Crosshair_Dot_M
  }
  n101 = StaticSwitchParameter(True=@n28, False=@n100.RGBA) {defaultValue=true, parameterName=EnableCrossHair Blur}
  NamedRerouteDeclaration(None=StaticSwitchParameter(True=Multiply(A=@n15, B=@n101), False=@n101) {parameterName=Use 0-1 Mask Mip Bleed Fix}) {name=Reticle Masks Pre, nodeColor=<rgba 1 0 0.4104 1>}
  n102 = VectorParameter() {
   group=Toggleable Reticle
   parameterName=DF Min Toggled
   sortPriority=100
   channelNames {a=Alpha Channel DF Min, b=Blue Channel DF Min, g=Green Channel DF Min, r=Red Channel DF Min}
   defaultValue=<rgba 0.5 0.5 0 0>
  }
  n103 = AppendVector(A=@n102, B=@n102)
  n104 = VectorParameter() {
   group=Distace Field Reticle
   parameterName=DF Min
   sortPriority=90
   channelNames {a=Alpha Channel DF Min, b=Blue Channel DF Min, g=Green Channel DF Min, r=Red Channel DF Min}
   defaultValue=<rgba 0.5 0.5 0 0>
  }
  n105 = AppendVector(A=@n104, B=@n104)
  n106 = ScalarParameter() {parameterName=ToggleThermalMask}
  n107 = VectorParameter() {
   group=Toggleable Reticle
   parameterName=DF Max Toggled
   sortPriority=101
   channelNames {a=Alpha Channel DF Max, b=Blue Channel DF Max, g=Green Channel DF Max, r=Red Channel DF Max}
   defaultValue=<rgba 0.5 0.5 0 0>
  }
  n108 = AppendVector(A=@n107, B=@n107)
  n109 = VectorParameter() {
   group=Distace Field Reticle
   parameterName=DF Max
   sortPriority=91
   channelNames {a=Alpha Channel DF Max, b=Blue Channel DF Max, g=Green Channel DF Max, r=Red Channel DF Max}
   defaultValue=<rgba 0.5 0.5 0 0>
  }
  n110 = AppendVector(A=@n109, B=@n109)
  n111 = NamedRerouteUsage()
  n112 = SmoothStep(
   Min=StaticSwitchParameter(True=LinearInterpolate(A=@n105, B=@n103, Alpha=@n106) {desc=Min DF}, False=@n105) {group=Toggleable Reticle, parameterName=Enable Toggleable Reticle, sortPriority=5}
   Max=StaticSwitchParameter(True=LinearInterpolate(A=@n110, B=@n108, Alpha=@n106) {desc=Max DF}, False=@n110) {group=Toggleable Reticle, parameterName=Enable Toggleable Reticle, sortPriority=5}
   Value=@n111
  )
  NamedRerouteDeclaration(None=StaticSwitchParameter(True=OneMinus(None=@n112), False=@n111) {group=Distace Field Reticle, parameterName=Use SDF}) {name=Reticle Masks Post, nodeColor=<rgba 0 0.1797 1 1>}
  NamedRerouteDeclaration(None=@n3.Camera - Eye Point Distance) {name=Camera - Eye point distance, nodeColor=<rgba 0 0.9766 1 1>}
  NamedRerouteDeclaration(None=LinearInterpolate(B=ScalarParameter() {defaultValue=3, group=Blur based on Accuracy, parameterName=MaxBlurMipBias}, Alpha=@n19)) {name=MipbiasBlur, nodeColor=<rgba 0 1 0.8125 1>}
  n113 = NamedRerouteUsage()
  n114 = Constant2Vector() {g=0.5, r=0.5}
  n115 = ScalarParameter() {defaultValue=1, group=Blur based on Accuracy, parameterName=CrossHairBlur_FadeMaskRadius}
  n116 = ScalarParameter() {group=Blur based on Accuracy, parameterName=CrossHairBlur_FadeMaskHarness}
  NamedRerouteDeclaration(None=Saturate(None=SphereMask(A=@n113, B=@n114, Radius=@n115, Hardness=@n116))) {name=CrosshairBlurMask, nodeColor=<rgba 0 0.9531 1 1>}
  NamedRerouteDeclaration(None=@n99.UVs) {name=0-1 UV Mask, nodeColor=<rgba 0.2188 0 1 1>}
  n117 = TextureSampleParameter2D(UVs=NamedRerouteUsage()) {
   group=Crosshair
   parameterName=T_Crosshair
   samplerSource=SSM_Clamp_WorldGroupSettings
   samplerType=SAMPLERTYPE_LinearColor
   texture=&Crosshair_Dot_M
  }
  n118 = Max(A=@n117.R, B=@n117.R)
  n119 = TextureSampleParameter2D(UVs=NamedRerouteUsage()) {
   group=Crosshair
   parameterName=T_Crosshair
   samplerSource=SSM_Clamp_WorldGroupSettings
   samplerType=SAMPLERTYPE_LinearColor
   texture=&Crosshair_Dot_M
  }
  n120 = Max(A=@n119.R, B=@n119.R)
  NamedRerouteDeclaration(None=AppendVector(A=AppendVector(A=@n118, B=@n120), B=Constant())) {name=Debug Reload Reticle Size, nodeColor=<rgba 0 1 0.5313 1>}
  n121 = MaterialFunctionCall(Float2=MaterialFunctionCall() {materialFunction=&ScreenResolution}.Visible Resolution) {materialFunction=&BreakOutFloat2Components}
  n122 = Divide(A=@n121.R, B=@n121.R)
  NamedRerouteDeclaration(None=AppendVector(A=@n122, B=Constant() {r=1})) {name=Screen Ratio, nodeColor=<rgba 1 0.1 0.1 1>}
  n123 = VectorParameter() {parameterName=CrosshairOffset, defaultValue=<rgba 0 0 0 1>}
  n124 = ScreenPosition()
  n125 = MaterialFunctionCall(Float2=Add(
    A=MaterialFunctionCall(X=@n123.R, Y=@n123.R) {materialFunction=&MakeFloat2}.Result
    B=Add(A=@n124.ViewportUV) {constB=-0.5}
   )) {materialFunction=&BreakOutFloat2Components}
  n126 = MaterialFunctionCall(X=Multiply(A=@n125.R, B=@n122), Y=@n125.G) {materialFunction=&MakeFloat2}
  NamedRerouteDeclaration(None=Add(A=NamedRerouteDeclaration(None=@n126.Result) {name=Screen Position Centered, nodeColor=<rgba 1 0.1 0.1 1>}) {constB=0.5}) {name=Screen Position Ratioed, nodeColor=<rgba 1 0.1 0.1 1>}
  n127 = SceneTexture(UVs=ScreenPosition().ViewportUV) {bFiltered=true, sceneTextureId=PPI_PostProcessInput0}
  n128 = SphereMask(A=NamedRerouteUsage(), B=Constant() {r=1}) {attenuationRadius=0.2, hardnessPercent=0}
  n129 = NamedRerouteUsage()
  n130 = ComponentMask(None=SceneTexture(UVs=NamedRerouteUsage()) {bFiltered=true, sceneTextureId=PPI_PostProcessInput0}.Color) {b=true, g=true, r=true}
  NamedRerouteDeclaration(None=LinearInterpolate(A=ComponentMask(None=@n127.Color) {b=true, g=true, r=true}, B=@n130, Alpha=Multiply(A=@n128, B=@n129))) {name=Scene Texture, nodeColor=<rgba 0 1 0.1797 1>}
  NamedRerouteDeclaration(None=Saturate(None=ScalarParameter() {desc=Set in C++, parameterName=TargetingWeight})) {name=ADS Transition, nodeColor=<rgba 1 0 0.5391 1>}
  NamedRerouteDeclaration(None=MaterialFunctionCall(UVs=ScreenPosition().ViewportUV, Texture Scale=@n6) {materialFunction=&ScaleUVsByCenter}.UVs) {name=ModifiedScreenPosition, nodeColor=<rgba 0 1 0.8594 1>}
  NamedRerouteUsage()
  NamedRerouteUsage()
  NamedRerouteUsage()
  StaticSwitchParameter(True=NamedRerouteUsage(), False=Multiply(A=NamedRerouteUsage(), B=NamedRerouteUsage())) {group=Debug, parameterName=DebugView, sortPriority=5}
  n131 = Divide(A=ComponentMask(None=ViewProperty() {property=MEVP_ViewSize}.Property) {g=true}) {constB=2000}
  n132 = Constant2Vector() {g=1, r=0.5}
  n133 = InverseLinearInterpolate(Value=ComponentMask(None=SceneTexture(UVs=@n132) {sceneTextureId=PPI_SceneDepth}.Color) {r=true}) {constA=10, constB=30}
  n134 = ScreenPosition()
  n135 = Saturate(None=Subtract(A=ComponentMask(None=@n134.ViewportUV) {g=true}) {constB=0.65})
  n136 = Power(Base=Sine(None=ComponentMask(None=@n134.ViewportUV) {r=true}) {period=2})
  n137 = LinearInterpolate(
   A=ScalarParameter() {defaultValue=0.001, group="DOF Blur (SM6 Only)", parameterName=Blur_DistanceMin}
   B=LinearInterpolate(
    A=ScalarParameter() {defaultValue=0.001, group="DOF Blur (SM6 Only)", parameterName=Blur_DistanceMax}
    B=ScalarParameter() {defaultValue=0.001, group="DOF Blur (SM6 Only)", parameterName=Blur_DistanceMaxPistols}
    Alpha=Saturate(None=@n133)
   )
   Alpha=NamedRerouteDeclaration(None=Multiply(A=@n135, B=@n136)) {name=Blur Gradient, nodeColor=<rgba 1 0 0.2812 1>}
  )
  MaterialExpressionMultiply_3 = Multiply(A=OneMinus(None=NamedRerouteUsage()), B=Multiply(A=Multiply(A=@n131, B=@n137), B=NamedRerouteUsage()))
  MaterialExpressionRound_0 = Round(None=ScalarParameter() {defaultValue=3, group="DOF Blur (SM6 Only)", parameterName=Blur Samples})
  MaterialExpressionMaterialFunctionCall_1 = MaterialFunctionCall() {materialFunction=&DitherTemporalAA}
  MaterialExpressionScalarParameter_7 = ScalarParameter() {defaultValue=1, group="DOF Blur (SM6 Only)", parameterName=TempAARotation}
  MaterialExpressionScalarParameter_8 = ScalarParameter() {defaultValue=4, group="DOF Blur (SM6 Only)", parameterName=DistanceMask}
  MaterialExpressionScalarParameter_14 = ScalarParameter() {defaultValue=1, group="DOF Blur (SM6 Only)", parameterName=TempAADistance}
  MaterialExpressionScalarParameter_15 = ScalarParameter() {defaultValue=0.618, group="DOF Blur (SM6 Only)", parameterName=Radial Offset}
  MaterialExpressionNamedRerouteUsage_15 = NamedRerouteUsage()
  MaterialExpressionSceneTexture_3 = SceneTexture() {sceneTextureId=PPI_PostProcessInput0}
  n138 = Custom!(
   Distance=@MaterialExpressionMultiply_3
   Steps=@MaterialExpressionRound_0
   Dither=@MaterialExpressionMaterialFunctionCall_1.Result
   TempAARotation=@MaterialExpressionScalarParameter_7
   DistanceMask=@MaterialExpressionScalarParameter_8
   TempAADistance=@MaterialExpressionScalarParameter_14
   RadialOffset=@MaterialExpressionScalarParameter_15
   UV=@MaterialExpressionNamedRerouteUsage_15
   SceneTexture=@MaterialExpressionSceneTexture_3.Color
  ) {
   code="float3 CurColor = .0;\r\nfloat RandomSamp = Dither; //Input\r\nfloat2 BaseUV = ClampSceneTextureUV(ViewportUVToSceneTextureUV(DERIV_BASE_VALUE(UV), 14), 14);\r\nfloat2 NewUV = BaseUV; //Input\r\nfloat CurDistance = .0;\r\nfloat2 CurOffset = .0;\r\n\r\n//int DistanceSteps = Steps + (int) RandomSamp;\r\n//int RadialSteps = Steps + (1 - (int) RandomSamp);\r\nint DistanceSteps = Steps;\r\nint RadialSteps = Steps;\r\nfloat StepSize = Distance / (int) DistanceSteps;\r\n\r\nfloat TwoPi = 6.283185;\r\nfloat Substep = .0;\r\nfloat2 ScenePixels=View.BufferSizeAndInvSize.xy*BaseUV;\r\n\r\n\r\nTempAARotation *= RandomSamp;\r\nTempAADistance *= StepSize * RandomSamp;\r\nfloat tempdist = -.0;\r\nfloat totaldist = .0;\r\n\r\nint i=0;\r\nif (DistanceSteps<1 || Distance <= .0)\r\n{\r\nreturn SceneTextureLookup(BaseUV, 14, 1).rgb;\r\n}\r\nelse\r\n{\r\nwhile ( i < (int) DistanceSteps)\r\n{\r\n\r\n//CurDistance+=StepSize;\r\nfor (int j = 0; j < (int) RadialSteps; j++)\r\n{\r\nCurOffset.x = cos(TwoPi*((TempAARotation+Substep) / RadialSteps));\r\nCurOffset.y = sin(TwoPi*((TempAARotation+Substep) / RadialSteps));\r\nCurOffset *=DistanceMask;\r\nNewUV.x = BaseUV.x + (CurOffset.x * (CurDistance+(RandomSamp*TempAADistance)));\r\nNewUV.y = BaseUV.y + (CurOffset.y * (CurDistance+(RandomSamp*TempAADistance)));\r\n\t\ttempdist = (length(NewUV-BaseUV)/Distance);\r\n\t\ttotaldist += tempdist;\r\n\r\nCurColor += SceneTextureLookup(NewUV, 14, 1).rgb*tempdist;\r\n\r\nSubstep++;\r\n}\r\nCurDistance+=StepSize;\r\nSubstep+=RadialOffset;\r\ni++;\r\n}\r\nCurColor = CurColor / totaldist;\r\nreturn CurColor;\r\n}"
   desc="\n*** Original source before expression upgrade ***\nfloat3 CurColor;\r\nfloat2 BaseUV = MaterialFloat2(ScreenAlignedPosition(Parameters.ScreenPosition).xy);\r\nfloat2 NewUV = BaseUV;\r\nfloat StepSize = Distance / (int) DistanceSteps;\r\nfloat CurDistance=0;\r\nfloat2 CurOffset;\r\nfloat TwoPi = 6.283185;\r\nfloat Substep;\r\nfloat2 ScenePixels=View.RenderTargetSize*BaseUV;\r\nScenePixels+=View.TemporalAAParams.r;\r\nfloat2 RandomSamp = ((uint)(ScenePixels.x) + 2 * (uint)(ScenePixels.y)) % 5;\r\nRandomSamp+=Texture2DSample(Tex,TexSampler,ScenePixels);\r\nRandomSamp/=5;\r\nRandomSamp-=0.5;\r\nTempAARotation*=RandomSamp;\r\nTempAADistance*=StepSize*RandomSamp;\r\nfloat tempdist = -0;\r\nfloat totaldist = 0;\r\n\r\nint i=0;\r\nif (DistanceSteps<1)\r\n{\r\nreturn DecodeSceneColorForMaterialNode(NewUV);\r\n}\r\nelse\r\n{\r\n//CurDistance += 0.5*StepSize;\r\nwhile ( i < (int) DistanceSteps)\r\n{\r\n\r\n//CurDistance+=StepSize;\r\nfor (int j = 0; j < (int) RadialSteps; j++)\r\n{\r\nCurOffset.x = cos(TwoPi*((TempAARotation+Substep) / RadialSteps));\r\nCurOffset.y = sin(TwoPi*((TempAARotation+Substep) / RadialSteps));\r\nCurOffset *=DistanceMask;\r\nNewUV.x = BaseUV.x + (CurOffset.x * (CurDistance+(RandomSamp*TempAADistance)));\r\nNewUV.y = BaseUV.y + (CurOffset.y * (CurDistance+(RandomSamp*TempAADistance)));\r\n\t\ttempdist = (length(NewUV-BaseUV)/Distance);\r\n\t\ttotaldist += tempdist;\r\n\r\nCurColor += DecodeSceneColorForMaterialNode(NewUV)*tempdist;\r\n//CurDistance+=(StepSize+(TempAADistance))/RadialSteps;\r\nSubstep++;\r\n}\r\nCurDistance+=StepSize;\r\nSubstep+=RadialOffset;\r\ni++;\r\n}\r\nCurColor = CurColor /totaldist;\r\nreturn CurColor;\r\n}\n*** Original source before expression upgrade ***\nfloat3 CurColor;\r\nfloat2 BaseUV = MaterialFloat2(ScreenAlignedPosition(Parameters.ScreenPosition).xy);\r\nfloat2 NewUV = BaseUV;\r\nfloat StepSize = Distance / (int) DistanceSteps;\r\nfloat CurDistance=0;\r\nfloat2 CurOffset;\r\nfloat TwoPi = 6.283185;\r\nfloat Substep;\r\nfloat2 ScenePixels=Frame.RenderTargetSize*BaseUV;\r\nScenePixels+=Frame.TemporalAAParams.r;\r\nfloat2 RandomSamp = ((uint)(ScenePixels.x) + 2 * (uint)(ScenePixels.y)) % 5;\r\nRandomSamp+=Texture2DSample(Tex,TexSampler,ScenePixels);\r\nRandomSamp/=5;\r\nRandomSamp-=0.5;\r\nTempAARotation*=RandomSamp;\r\nTempAADistance*=StepSize*RandomSamp;\r\nfloat tempdist = -0;\r\nfloat totaldist = 0;\r\n\r\nint i=0;\r\nif (DistanceSteps<1)\r\n{\r\nreturn DecodeSceneColorForMaterialNode(NewUV);\r\n}\r\nelse\r\n{\r\n//CurDistance += 0.5*StepSize;\r\nwhile ( i < (int) DistanceSteps)\r\n{\r\n\r\n//CurDistance+=StepSize;\r\nfor (int j = 0; j < (int) RadialSteps; j++)\r\n{\r\nCurOffset.x = cos(TwoPi*((TempAARotation+Substep) / RadialSteps));\r\nCurOffset.y = sin(TwoPi*((TempAARotation+Substep) / RadialSteps));\r\nCurOffset *=DistanceMask;\r\nNewUV.x = BaseUV.x + (CurOffset.x * (CurDistance+(RandomSamp*TempAADistance)));\r\nNewUV.y = BaseUV.y + (CurOffset.y * (CurDistance+(RandomSamp*TempAADistance)));\r\n\t\ttempdist = (length(NewUV-BaseUV)/Distance);\r\n\t\ttotaldist += tempdist;\r\n\r\nCurColor += DecodeSceneColorForMaterialNode(NewUV)*tempdist;\r\n//CurDistance+=(StepSize+(TempAADistance))/RadialSteps;\r\nSubstep++;\r\n}\r\nCurDistance+=StepSize;\r\nSubstep+=RadialOffset;\r\ni++;\r\n}\r\nCurColor = CurColor /totaldist;\r\nreturn CurColor;\r\n}"
   description=SceneTextureBlur
   inputs [
    {inputName=Distance, input {expression=&self:MaterialExpressionMultiply_3}}
    {inputName=Steps, input {expression=&self:MaterialExpressionRound_0}}
    {
     inputName=Dither
     input {expression=&self:MaterialExpressionMaterialFunctionCall_1}
    }
    {
     inputName=TempAARotation
     input {expression=&self:MaterialExpressionScalarParameter_7}
    }
    {
     inputName=DistanceMask
     input {expression=&self:MaterialExpressionScalarParameter_8}
    }
    {
     inputName=TempAADistance
     input {expression=&self:MaterialExpressionScalarParameter_14}
    }
    {
     inputName=RadialOffset
     input {expression=&self:MaterialExpressionScalarParameter_15}
    }
    {inputName=UV, input {expression=&self:MaterialExpressionNamedRerouteUsage_15}}
    {
     inputName=SceneTexture
     input {
      expression=&self:MaterialExpressionSceneTexture_3
      mask=1
      maskA=1
      maskB=1
      maskG=1
      maskR=1
     }
    }
   ]
  }
  StaticSwitchParameter(True=FeatureLevelSwitch(Default=@n130, SM6=@n138), False=@n130) {desc="DOF Blur is causing perf\r\nIssues on Xbox1", group="DOF Blur (SM6 Only)", parameterName=Enable DOF Blur, sortPriority=10}
  ViewProperty() {property=MEVP_PostVolumeUserFlags}
 ]
}
